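# Parent material for the ReptilianOcean back bling, inserting an animated fiery eye effect between the 2023 character PreFX and PostFX functions.
# Offset, radial UVs drive panned, rotated and distorted noise, carved by circular masks into inner and outer fire layers.
# These layers feed eye and iris colors, while masks blend in base color, spherical normals, roughness, metallic and emissive.
# A gradient mask texture modulated by two time-driven sines adds flickering outer glow.
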Material /Game/Accessories/FORT_Backpacks/Backpack_F_MED_ReptilianOcean/Materials/M_ReptilianOcean_Backpack_Parent
graph {
 properties {
  bUseMaterialAttributes=true
  bUsedWithMeshDeformer=true
  bUsedWithNanite=true
  bUsedWithSkeletalMesh=true
  pixelDepthOffsetMode=PDOM_Legacy
  shadingModel=MSM_FromMaterialExpression
  usedShadingModels=Default Lit | Subsurface | Clear Coat | Subsurface Profile | Cloth
 }
 nodes [
  n1 = VectorParameter() {parameterName=EyeShadowColor, defaultValue=<rgba 0.1 0 2 1>}
  n2 = LinearInterpolate(
   A=VectorParameter() {parameterName=EyeColor, defaultValue=<rgba 0.0833 0.7086 2 1>}.RGB
   B=@n1.RGB
   Alpha=Multiply(A=NamedRerouteUsage(), B=NamedRerouteUsage())
  )
  n3 = Add(A=NamedRerouteUsage(), B=NamedRerouteUsage())
  n4 = Multiply(A=Saturate(None=Multiply(A=@n3) {constB=2}), B=VectorParameter() {parameterName=IrisColor, defaultValue=<rgba 0 2.8741 15 1>}.RGB)
  Add(A=@n2, B=@n4)
  n5 = Multiply(A=Time(), B=ScalarParameter() {defaultValue=0.25, parameterName=GlowFlickerSpeed}) {constB=0.25}
  n6 = Add(A=Sine(None=@n5))
  n7 = Add(A=Sine(None=@n5) {period=0.5})
  Add(A=@n6, B=@n7)
  n8 = MaterialFunctionCall() {materialFunction=&MF_CharacterParent_2023_PreFX}
  ChannelMaskParameter(None=@n8.Tech Art Mask) {maskChannel=Blue, parameterName=EyesMask, defaultValue=<rgba 0 0 1 0>}
  MaterialFunctionCall(Material=NamedRerouteUsage(), M=@n8.M, Pre Skinned Local Position=@n8.M, Distance Blend=@n8.M, Fresnel=@n8.M) {materialFunction=&MF_CharacterParent_2023_PostFX}
  n9 = NamedRerouteUsage()
  n10 = ComponentMask(None=@n9) {r=true}
  n11 = Add(A=@n10)
  n12 = ComponentMask(None=@n9) {g=true}
  n13 = Multiply(A=AppendVector(A=LinearInterpolate(A=@n10, B=@n11, Alpha=@n12), B=@n12), B=Constant2Vector() {g=1, r=1})
  n14 = Subtract(A=MaterialFunctionCall(Color=NamedRerouteUsage()) {materialFunction=&SplitComponents}.R) {constB=0.5}
  n15 = NamedRerouteUsage()
  n16 = TextureSample!(
   UVs=Add(A=Panner(Coordinate=@n13) {speedX=-0.2, speedY=-0.2}, B=Multiply(A=@n14) {constB=0.1})
   Tex
   "DDX(UVs)"=DDX(Value=@n15)
   "DDY(UVs)"=DDY(Value=@n15)
   Apply View MipBias
  ) {
   mipValueMode=TMVM_Derivative
   samplerType=SAMPLERTYPE_LinearColor
   texture=&/T_ReptilianOcean_Mask
  }
  Multiply(A=@n16.G)
  n17 = TextureCoordinate()
  n18 = ComponentMask(None=VectorParameter() {parameterName=UVOffset, defaultValue=<rgba 0.195 0.2 0 1>}.RGB) {g=true, r=true}
  n19 = NamedRerouteDeclaration(None=Multiply(A=Subtract(A=@n17, B=@n18) {constB=0.25}) {constB=1.5}) {name=UVs, nodeColor=<rgba 0 1 0.8828 1>}
  n20 = TextureCoordinate()
  n21 = TextureSample(
   UVs=NamedRerouteDeclaration(None=MaterialFunctionCall(Vector or UVs=@n19) {materialFunction=&VectorToRadialValue}.Radial Coordinates) {name=RadialUV, nodeColor=<rgba 0 1 0.3906 1>}
   "DDX(UVs)"=DDX(Value=@n20)
   "DDY(UVs)"=DDY(Value=@n20)
  ) {
   mipValueMode=TMVM_Derivative
   samplerType=SAMPLERTYPE_LinearColor
   texture=&T_ReptilianOcean_Mask
  }
  NamedRerouteDeclaration(None=@n21.RGB) {name=StaticNoise, nodeColor=<rgba 0.1016 1 0 1>}
  n22 = Multiply(A=Power(Base=NamedRerouteUsage()) {constExponent=2.75}) {constB=5}
  n23 = Multiply(A=NamedRerouteUsage(), B=Constant2Vector() {g=2, r=3})
  n24 = Subtract(A=MaterialFunctionCall(Color=NamedRerouteUsage()) {materialFunction=&SplitComponents}.R) {constB=0.5}
  n25 = NamedRerouteUsage()
  n26 = TextureSample(
   UVs=Add(A=Panner(Coordinate=@n23) {speedX=-0.1, speedY=-0.2}, B=Multiply(A=@n24) {constB=0.15})
   "DDX(UVs)"=DDX(Value=@n25)
   "DDY(UVs)"=DDY(Value=@n25)
  ) {
   mipValueMode=TMVM_Derivative
   samplerType=SAMPLERTYPE_LinearColor
   texture=&T_ReptilianOcean_Mask
  }
  n27 = OneMinus(None=@n26.G)
  n28 = Saturate(None=Saturate(None=Subtract(A=@n22, B=@n27)))
  n29 = Multiply(A=NamedRerouteUsage(), B=Constant() {r=0.92}) {constB=0.95}
  n30 = SmoothStep(Value=@n29) {constMax=0.575, constMin=0.525}
  NamedRerouteDeclaration(None=Saturate(None=Subtract(A=@n28, B=@n30))) {name=OutterFire, nodeColor=<rgba 1 0 0.5859 1>}
  n31 = NamedRerouteUsage()
  n32 = OneMinus(None=@n16.G)
  n33 = Subtract(A=Multiply(A=Power(Base=@n31) {constExponent=1.7}) {constB=3}, B=@n32)
  n34 = Add(A=Multiply(A=NamedRerouteUsage()) {constB=3.75}, B=Constant2Vector() {g=0.5, r=0.5})
  n35 = Subtract(A=MaterialFunctionCall(Color=NamedRerouteUsage()) {materialFunction=&SplitComponents}.R) {constB=0.5}
  n36 = TextureSample!(UVs=Add(A=Rotator(Coordinate=@n34) {speed=-0.5}, B=Multiply(A=@n35) {constB=0.05}), Tex, Apply View MipBias) {
   samplerSource=SSM_Clamp_WorldGroupSettings
   samplerType=SAMPLERTYPE_LinearColor
   texture=&T_ReptilianOcean_Mask
  }
  n37 = Multiply(A=Power(Base=@n36.B) {constExponent=0.75}) {constB=2}
  n38 = Saturate(None=Multiply(A=Subtract(A=@n33, B=@n37), B=SmoothStep(Value=@n29) {constMin=0.55}) {constB=2})
  n39 = NamedRerouteUsage()
  n40 = Constant() {r=0.625}
  n41 = SmoothStep(Min=@n32, Value=SmoothStep(Value=Multiply(A=@n39, B=@n40) {constB=0.95}) {constMax=0.525, constMin=0.475})
  NamedRerouteDeclaration(None=Saturate(None=Subtract(A=@n38, B=@n41))) {name=InnerFire, nodeColor=<rgba 1 0.3047 0 1>}
  n42 = GetMaterialAttributes(None=NamedRerouteUsage()) {attributeGetTypes [
    #1
    #2
    #3
    #4
    #5
   ]}
  NamedRerouteDeclaration(None=@n42.Emissive Color) {name=inEmissive, nodeColor=<rgba 0 1 0.2031 1>}
  n43 = NamedRerouteUsage()
  n44 = Add(A=Multiply(A=@n2) {constB=0.4}, B=@n4)
  n45 = NamedRerouteUsage()
  n46 = NamedRerouteUsage()
  n47 = ScalarParameter() {defaultValue=3, parameterName=OutterGlowSpread}
  n48 = NamedRerouteUsage()
  n49 = Saturate(None=Multiply(A=Power(Base=@n46, Exp=@n47), B=OneMinus(None=@n48)))
  n50 = Multiply(A=NamedRerouteUsage(), B=ScalarParameter() {defaultValue=0.15, parameterName=OutterGlowIntensity}) {constB=0.15}
  NamedRerouteDeclaration(None=Add(A=LinearInterpolate(A=@n43, B=@n44, Alpha=@n45), B=Multiply(A=@n49, B=@n50))) {name=outEmissive, nodeColor=<rgba 0.0312 1 0 1>}
  NamedRerouteDeclaration(None=@n1.RGB) {name=EyeColor, nodeColor=<rgba 1 0.7266 0 1>}
  n51 = TextureSample() {
   samplerType=SAMPLERTYPE_LinearColor
   texture=/Game/Accessories/FORT_Backpacks/Backpack_F_MED_ReptilianOcean/Textures/T_ReptilianOcean_Backpack_FX
  }
  NamedRerouteDeclaration(None=@n51.B) {name=EyeBallMask, nodeColor=<rgba 0 1 0.6016 1>}
  NamedRerouteDeclaration(None=@n8.Result) {name=inMaterialAttributes, nodeColor=<rgba 1 0.0234 0 1>}
  n52 = SetMaterialAttributes!(
   MaterialAttributes=@n42.MaterialAttributes
   Base Color=LinearInterpolate(A=@n42.Base Color, B=NamedRerouteUsage(), Alpha=NamedRerouteUsage())
   Emissive Color=NamedRerouteUsage()
   Normal=LinearInterpolate(A=@n42.Normal, B=NamedRerouteUsage(), Alpha=NamedRerouteUsage())
   Roughness=LinearInterpolate(A=@n42.Roughness, Alpha=NamedRerouteUsage()) {constB=0.1}
   Metallic=LinearInterpolate(A=@n42.Metallic, Alpha=NamedRerouteUsage()) {constB=0}
  ) {attributeSetTypes [
    #1
    #2
    #3
    #4
    #5
   ]}
  NamedRerouteDeclaration(None=@n52) {name=outMaterialAttributes, nodeColor=<rgba 0.5938 0 1 1>}
  NamedRerouteDeclaration(None=@n2) {name=outBaseColor, nodeColor=<rgba 0 0.7656 1 1>}
  n53 = NamedRerouteUsage()
  n54 = ScalarParameter() {defaultValue=1, parameterName=EyesNormalsIntensity}
  NamedRerouteDeclaration(None=DeriveNormalZ(InXY=Multiply(A=@n53, B=@n54))) {name=SphericalNormals, nodeColor=<rgba 0.1953 1 0 1>}
  n55 = TextureSample() {
   samplerType=SAMPLERTYPE_LinearGrayscale
   texture=/Game/Accessories/FORT_Backpacks/Backpack_F_MED_ReptilianOcean/Textures/T_ReptilianOcean_Backpack_TechArtFX_Mask
  }
  n56 = Saturate(None=Multiply(A=@n6, B=@n7))
  NamedRerouteDeclaration(None=Multiply(
    A=NamedRerouteDeclaration(None=@n55.R) {name=GradientMaskTexture, nodeColor=<rgba 0 0.7422 1 1>}
    B=LinearInterpolate(Alpha=@n56) {constA=0.8}
   ) {constB=0.5}) {name=Glow, nodeColor=<rgba 0.9688 1 0 1>}
  NamedRerouteDeclaration(None=SphereMask(A=NamedRerouteUsage(), B=Constant()) {attenuationRadius=0.3, hardnessPercent=0}) {name=BaseCircle, nodeColor=<rgba 0.9453 1 0 1>}
 ]
}
